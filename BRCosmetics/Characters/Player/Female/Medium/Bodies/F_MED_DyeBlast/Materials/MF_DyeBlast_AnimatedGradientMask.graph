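# Material function for the DyeBlast female medium body that outputs an animated gradient mask, with an optional stepped debug view.
# MF_CharacterGradients builds the base gradient using a texture or sphere arm mask; both Radial/Linear switch branches use the radial output.
# The Animated branch offsets a threshold with scrolling local-space cloud noise, shaping a sharp-edged band with a falloff.
# Timestamp parameters feed vertex-interpolated ramps declared as PannerTime reroutes.
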MaterialFunction /BRCosmetics/Characters/Player/Female/Medium/Bodies/F_MED_DyeBlast/Materials/MF_DyeBlast_AnimatedGradientMask
graph {nodes [
  Frac(None=Multiply(A=Time()) {constB=0.4})
  n1 = NamedRerouteUsage()
  n2 = StaticBoolParameter() {defaultValue=true, group=1_Features, parameterName=Clamp Debug View, sortPriority=34}
  n3 = ScalarParameter() {defaultValue=10, parameterName=DebugViewSteps, sortPriority=500}
  FunctionOutput(None=StaticSwitchParameter(
    True=MaterialFunctionCall!(Gradient Input=@n1, Clamp Input=@n2, Num Steps=@n3) {materialFunction=/Game/Global/MaterialFunctions/TechArt/MF_GradientSteppedDebugView}.Result
    False=NamedRerouteUsage()
   ) {group=1_Features, parameterName=Debug View, sortPriority=33}) {outputName=AnimatedGradientMask}
  Multiply(A=ScalarParameter() {defaultValue=0.64, parameterName=TimeTest})
  n4 = PreSkinnedPosition()
  n5 = SphereMask(
   A=ComponentMask(None=VertexInterpolator(VS=@n4).PS) {g=true, r=true}
   B=Constant2Vector()
   Radius=ScalarParameter() {defaultValue=30, parameterName=CharacterGradient_RumpRadius, sortPriority=1000}
   Hardness=ScalarParameter() {defaultValue=0.8, parameterName=CharacterGradient_RumpFalloffHardness, sliderMax=1, sortPriority=1000}
  ) {attenuationRadius=30, hardnessPercent=80}
  n6 = TextureSampleParameter2D() {
   parameterName=Arms Mask
   samplerType=SAMPLERTYPE_Masks
   texture=/Game/Characters/Player/Male/Medium/Bodies/M_MED_ZirconSweep/Textures/T_M_MED_ZirconSweep_Body_TechArtFX_Masks
  }
  n7 = MaterialFunctionCall!(
   Local Position
   ArmBalance=ScalarParameter() {defaultValue=0.5, parameterName=Arm Balance, sliderMax=1.5, sortPriority=-10}
   RumpBalance=ScalarParameter() {defaultValue=0.5, parameterName=Rump Balance, sliderMax=1.5, sortPriority=-10}
   Arms Mask=StaticSwitchParameter(True=OneMinus(None=@n5), False=@n6.G) {group=1_Features, parameterName=Use Cylinder Mask, sortPriority=320}
   Arm_GradientEndPosition
   Arm_GradientBeginPosition
   Arm_GradientDirection
   Rump_GradientEndHeight=ScalarParameter() {parameterName=RumpGradient_FeetHeight, sliderMax=180, sortPriority=64}
   Rump_GradientBeginHeight=ScalarParameter() {defaultValue=140, parameterName=RumpGradient_TopHeight, sliderMax=180, sortPriority=64}
   Radial_CenterHeight=ScalarParameter() {defaultValue=0.5, parameterName=Radial_CenterHeightPercent, sliderMax=1, sortPriority=16}
   Radial_HorizontalCompression=ScalarParameter() {defaultValue=1, parameterName=Radial_HorizontalCompression, sliderMax=2, sliderMin=0.01, sortPriority=16}
   Disable Arm Gradients=StaticBoolParameter() {group=1_Features, parameterName=Disable Arm Gradient}
   Disable Rump Gradient=StaticBoolParameter() {group=1_Features, parameterName=Disable Rump Gradient}
  ) {materialFunction=/Game/Global/MaterialFunctions/TechArt/MF_CharacterGradients}
  NamedRerouteDeclaration(None=StaticSwitchParameter(True=@n7.RadialGradient, False=@n7.RadialGradient) {group=1_Features, parameterName="Radial (true) or Linear (false)"}) {name=CharacterGradient, nodeColor=<rgba 1 0.4219 0 1>}
  n8 = NamedRerouteUsage()
  n9 = NamedRerouteUsage()
  n10 = Divide(A=Time() {bOverride_Period=true, period=8}) {constB=8}
  n11 = Multiply(
   A=LocalPosition() {localOrigin=InstancePreSkinning}.XYZ
   B=VectorParameter() {parameterName=NoiseTextureScale, defaultValue=<rgba 0.02 0.02 0.02 0.02>}.RGB
  ) {constB=0.02}
  n12 = Add(A=AppendVector(A=@n10, B=@n10), B=ComponentMask(None=VertexInterpolator(VS=@n11).PS) {b=true, r=true})
  n13 = TextureObject() {texture=&AtmosphericCloudNoise01}
  n14 = Multiply(A=ConstantBiasScale(None=TextureSample(UVs=@n12, Tex=@n13).R) {bias=-0.5, scale=2}) {constB=0.2}
  n15 = Subtract(A=Subtract(A=Multiply(A=@n8) {constB=1.25}) {constB=0.25}, B=Multiply(A=Add(A=@n9, B=@n14)) {constB=0.8})
  n16 = Saturate(None=Multiply(A=@n15) {constB=2})
  n17 = Add(A=Multiply(A=@n15) {constB=11})
  n18 = Multiply(A=Power(Base=OneMinus(None=@n16)) {constExponent=12}, B=Power(Base=Saturate(None=@n17)))
  n19 = NamedRerouteUsage()
  NamedRerouteDeclaration(None=Saturate(None=StaticSwitchParameter(True=@n18, False=@n19) {group=1_Features, parameterName=Animated, sortPriority=40})) {name=AnimatedGradient, nodeColor=<rgba 0.1016 0 1 1>}
  NamedRerouteDeclaration(None=ScalarParameter() {group=RumbleBeat_TechArt, parameterName=TimestampPanner}) {name=TimestampPanner, nodeColor=<rgba 0 0.3672 1 1>}
  NamedRerouteDeclaration(None=ScalarParameter() {group=RumbleBeat_TechArt, parameterName=TransitionLengthPanner}) {name=TransitionLengthPanner, nodeColor=<rgba 0.6875 1 0 1>}
  n20 = NamedRerouteUsage()
  n21 = NamedRerouteUsage()
  NamedRerouteDeclaration(None=VertexInterpolator(VS=MaterialFunctionCall(TimestampA=@n20, TransitionLength_In=@n21) {materialFunction=&MF_Timestamp_InOut_01}.Linear).PS) {name=PannerTimeB, desc=ramp down, nodeColor=<rgba 1 0.965 0.0428 1>}
  NamedRerouteDeclaration(None=ScalarParameter() {group=RumbleBeat_TechArt, parameterName=TimestampB}) {name=TimestampB, nodeColor=<rgba 1 0.0563 0.4827 1>}
  NamedRerouteDeclaration(None=ScalarParameter() {parameterName=NumberOfKills}) {name=NumberOfKills, nodeColor=<rgba 1 0 0.9141 1>}
  NamedRerouteDeclaration(None=ScalarParameter() {group=RumbleBeat_TechArt, parameterName=TransitionLength_InB}) {name=TransitionLength_InB, nodeColor=<rgba 0.0704 1 0.2285 1>}
  NamedRerouteDeclaration(None=ScalarParameter() {group=RumbleBeat_TechArt, parameterName=ToggleGlow}) {name=ToggleGlow, nodeColor=<rgba 0.0312 0 1 1>}
  n22 = NamedRerouteUsage()
  n23 = NamedRerouteUsage()
  NamedRerouteDeclaration(None=VertexInterpolator(VS=MaterialFunctionCall(TimestampA=@n22, TransitionLength_In=@n23) {materialFunction=&MF_Timestamp_InOut_01}.Linear).PS) {name=PannerTime, desc=ramp up, nodeColor=<rgba 0.0274 0.2424 1 1>}
  OneMinus(None=@n8)
  VectorParameter() {parameterName=RumpGradientStart, defaultValue=<rgba 0 8 140 0>}
 ]}
