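# Material function for the female medium DyeBlast player body that adds animated emissive edge lines and glow to incoming material attributes.
# MF_DyeBlast_AnimatedGradientMask multiplies the red and green channels of the body's UV-lines texture to form separate edge and glow masks.
# Each mask gets its own tint, intensity and saturation parameters before the two are summed.
# A lerp feeds the summed result into emissive output, and the function also takes a tech-art mask input.
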MaterialFunction /BRCosmetics/Characters/Player/Female/Medium/Bodies/F_MED_DyeBlast/Materials/MF_DyeBlastBody
graph {nodes [
  n1 = GetMaterialAttributes(None=NamedRerouteUsage()) {attributeGetTypes [#1, #2]}
  FunctionOutput(None=SetMaterialAttributes!(MaterialAttributes=@n1.MaterialAttributes, EmissiveColor=NamedRerouteUsage()) {attributeSetTypes [#2]})
  NamedRerouteDeclaration(None=FunctionInput() {inputName=In_MaterialAttributes, inputType=FunctionInput_MaterialAttributes}) {name=In_MaterialAttributes, nodeColor=<rgba 0.5938 1 0 1>}
  NamedRerouteDeclaration(None=FunctionInput() {inputName=In_TechArtMask, sortPriority=1}) {name=In_TechArtMask, nodeColor=<rgba 1 0 0.9141 1>}
  NamedRerouteDeclaration(None=@n1.BaseColor) {name=In_BaseColor, nodeColor=<rgba 1 0.5156 0 1>}
  NamedRerouteDeclaration(None=@n1.EmissiveColor) {name=In_EmissiveColor, nodeColor=<rgba 0.1953 0 1 1>}
  n2 = MaterialFunctionCall() {materialFunction=&/MF_DyeBlast_AnimatedGradientMask}
  n3 = NamedRerouteDeclaration(None=Multiply(
    A=@n2.AnimatedGradientMask
    B=TextureSampleParameter2D() {parameterName=UVEdges, texture=/BRCosmetics/Characters/Player/Female/Medium/Bodies/F_MED_DyeBlast/Textures/T_DyeBlast_Body_UVLines}.R
   )) {name=EmissiveEdgesMask, nodeColor=<rgba 1 0 0.7969 1>}
  n4 = ScalarParameter() {defaultValue=1, parameterName=EmissiveEdgesMultiplier}
  n5 = Desaturation(
   None=Multiply(A=Multiply(A=@n3, B=@n4) {constB=12}, B=VectorParameter() {parameterName=EmissiveEdgesColor, defaultValue=<rgba 0.7157 0.0931 0.4621 1>}.RGB)
   Fraction=OneMinus(None=ScalarParameter() {defaultValue=1, parameterName=EmissiveEdgesSaturation})
  )
  n6 = NamedRerouteDeclaration(None=Multiply(
    A=@n2.AnimatedGradientMask
    B=TextureSampleParameter2D() {parameterName=UVEdges, texture=&T_DyeBlast_Body_UVLines}.G
   )) {name=EmissiveGlowMask, nodeColor=<rgba 0.0595 1 0.3014 1>}
  n7 = ScalarParameter() {defaultValue=1, parameterName=EmissiveGlowMultiplier}
  n8 = Desaturation(
   None=Multiply(A=Multiply(A=@n6, B=@n7) {constB=12}, B=VectorParameter() {parameterName=EmissiveGlowColor, defaultValue=<rgba 0.7157 0.0931 0.4621 1>}.RGB)
   Fraction=OneMinus(None=ScalarParameter() {defaultValue=1, parameterName=EmissiveGlowSaturation})
  )
  NamedRerouteDeclaration(None=LinearInterpolate(A=NamedRerouteUsage(), B=Add(A=@n5, B=@n8), Alpha=NamedRerouteUsage())) {name=Out_EmissiveColor, nodeColor=<rgba 0.125 1 0 1>}
 ]}
