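# Physics asset supplying cape collision volumes for the female medium DyeBlast player character body, as its name indicates.
# Capsules cover the pelvis, spine chain, clavicles and neck, with a doubled capsule on the upper spine and a sphere on the head.
# Limb bodies for arms, legs and feet exist without shapes, and there are no constraints, so it serves as collision only rather than a ragdoll.
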PhysicsAsset /BRCosmetics/Characters/Player/Female/Medium/Bodies/F_MED_DyeBlast/Meshes/F_MED_DyeBlast_CapeColliders
bodies {
 calf_l []
 calf_r []
 clavicle_l [{shapeType=Capsule, length=10.0998, radius=8.8065, center=<9.3628 -0.3332 -2.4818>, rotation=<rot 57.7784 1.5946 179.7404>}]
 clavicle_r [{shapeType=Capsule, length=9.7396, radius=9.1667, center=<-9.2997 -0.5047 2.4659>, rotation=<rot 59.7787 1.6709 -0.476>}]
 foot_l []
 foot_r []
 head [{shapeType=Sphere, radius=17.3125, center=<11.028 -0.174 0>}]
 lowerarm_l []
 lowerarm_r []
 neck_02 [{shapeType=Capsule, length=13.4375, radius=7.3175, center=<-0.2642 -0.3408 0>, rotation=<rot 90 -171.1193 -175.3978>}]
 pelvis [{shapeType=Capsule, length=7.9882, radius=14.9162, center=<-2.4887 -0.8382 0>}]
 spine_01 [{shapeType=Capsule, length=6.3022, radius=11.5403, center=<-1.7965 0.1884 0>, rotation=<rot 0 25.6223 0>}]
 spine_02 [{shapeType=Capsule, length=5.1811, radius=8.8683, center=<-2.1109 0.6795 0.0001>, rotation=<rot 0 16.894 0>}]
 spine_03 [{shapeType=Capsule, length=6.6939, radius=7.3487, center=<-1.3044 1.5894 0.0001>}]
 spine_04 [{shapeType=Capsule, length=9.8399, radius=12.3719, center=<2.2712 3.0925 0.0002>}]
 spine_05 [
  {shapeType=Capsule, length=10.125, radius=12.7499, center=<-1.177 3.249 0>}
  {shapeType=Capsule, length=14.4941, radius=13.4002, center=<2.3185 3.6146 0>}
 ]
 thigh_l []
 thigh_r []
 upperarm_l []
 upperarm_r []
}
constraints []
