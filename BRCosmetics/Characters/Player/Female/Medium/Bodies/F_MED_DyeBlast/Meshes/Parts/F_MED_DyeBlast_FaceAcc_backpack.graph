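# Physics asset for the DyeBlast medium female face accessory mesh (backpack variant), defining ragdoll-style bodies for its dynamic back-hair bones.
# Capsules on five dyn_back_hair bones chain through swing-limited, twist-locked constraints to the head, branching from hair_01 to 02 and 04-05, with hair_03 attached separately.
# Unconstrained capsules on head, neck, upper spine, right clavicle and right upper arm form collision volumes around the hair.
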PhysicsAsset /BRCosmetics/Characters/Player/Female/Medium/Bodies/F_MED_DyeBlast/Meshes/Parts/F_MED_DyeBlast_FaceAcc_backpack
bodies {
 clavicle_r [{
   shapeType=Capsule
   length=10.5002
   radius=8.0785
   shapeName=clavicle_r_capsule
   center=<-7.1632 0 -2.4931>
   rotation=<rot -86.81 0.0012 -0.0012>
  }]
 dyn_back_hair_01 [{shapeType=Capsule, length=2.5176, radius=4.9824, shapeName=dyn_back_hair_01_capsule, center=<2.2 0 0>, rotation=<rot -90 0 0>}]
 dyn_back_hair_02 [{shapeType=Capsule, length=5.5269, radius=3.9712, shapeName=dyn_back_hair_02_capsule, center=<5 0 0>, rotation=<rot -90 0 0>}]
 dyn_back_hair_03 [{shapeType=Capsule, length=3, radius=2.5, shapeName=dyn_back_hair_03_capsule, center=<2.3 0 0>, rotation=<rot -77.5177 -0.0002 0.0003>}]
 dyn_back_hair_04 [{shapeType=Capsule, length=1.6097, radius=4.5903, shapeName=dyn_back_hair_04_capsule, center=<1.35 0 0>, rotation=<rot -90 7.9325 0>}]
 dyn_back_hair_05 [{shapeType=Capsule, length=5.8854, radius=2.1146, shapeName=dyn_back_hair_05_capsule, center=<2 0 0>, rotation=<rot -90 0 0>}]
 head [{shapeType=Capsule, length=0, radius=9.574, shapeName=head_capsule, center=<6.5399 0.7225 0>}]
 neck_01 [
  {shapeType=Capsule, length=17.7217, radius=7.5495, shapeName=neck_01_capsule, center=<0.0239 2.9535 0.002>, rotation=<rot 0 -90 85.2507>}
  {shapeType=Capsule, length=4.5623, radius=4.7164, shapeName=neck_01_capsule, center=<3.4095 -0.8668 0>}
 ]
 spine_05 [
  {shapeType=Capsule, length=6.8658, radius=4.3353, shapeName=spine_05_capsule, center=<9.1653 -13.2644 -0.0326>, rotation=<rot 90 0 0>}
  {shapeType=Capsule, length=15, radius=15, shapeName=spine_05_capsule, center=<0 2 0>}
  {shapeType=Capsule, length=7.3785, radius=8.6435, shapeName=spine_05_capsule, center=<6.4361 -16.7199 0>}
 ]
 upperarm_r [{
   shapeType=Capsule
   length=15.3722
   radius=7.9529
   shapeName=upperarm_r_capsule
   center=<-10.8561 0.2554 0.6654>
   rotation=<rot 0.1014 -95.2545 92.7326>
  }]
}
constraints [
 {bone1Name=dyn_back_hair_05, bone2Name=dyn_back_hair_04, swing1LimitDegrees=35, swing2LimitDegrees=35, twistLimitDegrees=0}
 {bone1Name=dyn_back_hair_04, bone2Name=dyn_back_hair_01, swing1LimitDegrees=30, swing2LimitDegrees=30, twistLimitDegrees=0}
 {bone1Name=dyn_back_hair_02, bone2Name=dyn_back_hair_01, swing1LimitDegrees=20, swing2LimitDegrees=20, twistLimitDegrees=0}
 {bone1Name=dyn_back_hair_01, bone2Name=head, swing1LimitDegrees=25, swing2LimitDegrees=25, twistLimitDegrees=0}
 {bone1Name=dyn_back_hair_03, bone2Name=head, swing1LimitDegrees=40, swing2LimitDegrees=40, twistLimitDegrees=0}
]
