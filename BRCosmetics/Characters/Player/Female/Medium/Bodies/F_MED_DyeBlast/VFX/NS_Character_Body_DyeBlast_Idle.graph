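# Idle Niagara effect for the DyeBlast female medium body cosmetic, floating twelve glowing geometric shapes around character sockets.
# A single GeoShapes emitter bursts persistent particles, partitioned into three groups that each follow a socket-reader socket.
# Index-selected arrays set scale and offset, while age-driven sines add drift and pulse per-partition colors.
# Two mesh renderers and a sprite renderer draw them, with sprites aligned to mesh orientation and a derived rotated line orientation.
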NiagaraSystem /BRCosmetics/Characters/Player/Female/Medium/Bodies/F_MED_DyeBlast/VFX/NS_Character_Body_DyeBlast_Idle
dependencies {
 usedDataInterfaces [
  &NiagaraDataInterfaceArrayFloat
  &NiagaraDataInterfaceArrayFloat3
  &NiagaraDataInterfaceSocketReader
 ]
 usedDynamicInputs []
 usedModules [
  &self:SystemSpawnScript.SystemScriptSource.SystemScriptGraph.NiagaraNodeAssignment_0.SetVariables_#1
  &SystemState
 ]
}
emitters {GeoShapes {
  emitterUpdateScript [
   &EmitterState(
    Inactive Response=::NewEnumerator1
    Life Cycle Mode=::NewEnumerator1
    Loop Behavior=::NewEnumerator1
    Loop Duration Mode=::NewEnumerator1
    UseLoopDelay=false
   )
   &SpawnBurst_Instantaneous(Spawn Count=12, Use Spawn Probability=false, UseLoopCountLimit=false)
  ]
  particleSpawnScript [
   &PartitionParticles(Number Of Partitions=3, Particles Per Partition=4, Partition Mode:ENiagara_PartitionMode_v2=::NewEnumerator1)
   &InitializeParticle(
    Color=<rgba 0.9235 0.1531 2 1>
    Mesh Scale=&Multiply_VectorByFloat(
     Vector=<0.1 0.1 0.3>
     Float=&SelectFloatFromArray(
      Direct Array Index=&ReturnExecIndex()
      Float Selection Array=DataInterface() {propertyValues {FloatData [0.25,0.1,0.2,0.125,0.25,0.1,0.2,0.1,0.1,0.1,0.25,0.2]}}
     )
    )
    Mesh Scale Mode=::NewEnumerator4
    Sprite Size=<100 100>
    Sprite Size Mode=::NewEnumerator4
    UsePositionOffset=false
    Write Material Random=true
   )
   /SetVariables(Particles.SpriteSize=&Multiply_Vector2D_ByFloat(
     Float=350
     Vector2D=&MakeVector2D(
      X=&MakeFloatFromVector(Vector=$Particles.Scale, Channel=::NewEnumerator0)
      Y=&MakeFloatFromVector(Vector=$Particles.Scale, Channel=::NewEnumerator2)
     )
    )) {isSetParameters=true}
   &InitialMeshOrientation(
    Mesh Axis To Orient=<0 0 1>
    Mesh Orientation Mode=::NewEnumerator2
    Orientation Coordinate Space=::Local
    Orientation Vector=&MakeVector(
     X=&SelectFloatFromArray(
      Array Sampling Mode=::NewEnumerator0
      Direct Array Index=?
      Float Selection Array=DataInterface() {propertyValues {FloatData [-1,1]}}
      Random Seed=0
     )
     Y=&RandomRangeFloat(Minimum=-0.3, Maximum=0.3)
     Z=&RandomRangeFloat(Minimum=-0.5, Maximum=0.5)
    )
    Use Rotation Vector=false
   )
   &DynamicMaterialParameters(
    Index 0 Param 1=1
    Index 0 Param 2=150
    Module.Index0Float0Write=true
    Module.Index0Float1Write=true
    Module.Index0Float2Write=true
    Module.Index0Float3Write=true
   )
  ]
  particleUpdateScript [
   &ParticleState(Kill Particles When Lifetime Has Elapsed=false, Let Infinitely Lived Particles Die When Emitter Deactivates=true)
   &SocketLocation(
    Orientation Sampling=::NewEnumerator1
    SamplePosOffset=true
    Sampled Orientation Coordinate Space:ENiagaraCoordinateSpace=::World
    Sampled Position Offset=&Add_Vector(
     A=&MergeHelper_VectorSwitch(0=<0 0 0>, 1=<0 0 -5>, 2=<0 0 1>, 3=?, Index=$Particles.PartitionParticles.Partition, Num=::NewEnumerator1)
     B=&Add_Vector(
      A=&MakeVector(
       X=&Sine(Normalized Angle=$Emitter.Age, Period=&RandomRangeFloat(Minimum=2, Maximum=4), Scale=1, Bias=0)
       Y=&Sine(Normalized Angle=$Emitter.Age, Period=&RandomRangeFloat(Minimum=2, Maximum=4), Scale=1, Bias=0)
       Z=&Sine(Normalized Angle=$Emitter.Age, Period=&RandomRangeFloat(Minimum=2, Maximum=4), Scale=1, Bias=0)
      )
      B=&SelectVectorFromArray(
       Direct Array Index=&ReturnExecIndex()
       Vector Selection Array=DataInterface() {propertyValues {FloatData [
          <-5 -2.5 8.9524>
          <-10 5 12>
          <-10 0 4.3829>
          <-8.8238 -7.5 0>
          <10 0 -5>
          <12 -1.5 12>
          <7.5 5 17.5>
          <10.5 5 5>
          <-12.5 3 10>
          <-8 -2 25>
          <-2.5 -7.5 15>
          <-5 -6.9374 0.0839>
         ]}}
      )
     )
    )
    Sampled Position Offset Coordinate Space:ENiagaraSkelMeshTransforms=::NewEnumerator2
    Socket Index=$Particles.PartitionParticles.Partition
    Socket Reader=$System.Socket Reader
    Socket Sampling Mode=::NewEnumerator3
    VelocitySpeedLimit=false
   )
   /SetVariables(
    Particles.GlowColor=&MakeLinearColorFromVectorAndFloat("Vector (RGB)"=&MakeVectorFromLinearColorRGB(Color=$Particles.Initial.Color), "Float (Alpha)"=1)
    Particles.GlowLineColor:LinearColor=&MakeLinearColorFromVectorAndFloat(
     "Vector (RGB)"=&Multiply_VectorByFloat(
      Vector=&Add_Vector(
       A=&MakeVector(
        X=&Sine(Normalized Angle=$Emitter.Age, Period=&RandomRangeFloat(Minimum=5, Maximum=2.5), Scale=1, Bias=1)
        Y=&Sine(Normalized Angle=$Emitter.Age, Period=&RandomRangeFloat(Minimum=0.5, Maximum=1), Scale=1, Bias=0.1)
        Z=&Sine(Normalized Angle=$Emitter.Age, Period=&RandomRangeFloat(Minimum=1, Maximum=2), Scale=1, Bias=0.5)
       )
       B=&MergeHelper_VectorSwitch(
        0=&RandomRangeVector(Minimum=<0.75 1.25 1.1>, Maximum=<0.5 2.5 1.25>)
        1=&RandomRangeVector(Minimum=<1.25 1.2 1>, Maximum=<1.1 1.25 1>)
        2=&RandomRangeVector(Minimum=<1.2 0.75 1.5>, Maximum=<1.25 0.5 2>)
        3=?
        Index=$Particles.PartitionParticles.Partition
        Num=::NewEnumerator1
       )
      )
      Float=&Sine(Normalized Angle=$Emitter.Age, Period=&RandomRangeFloat(Minimum=2, Maximum=4), Scale=1, Bias=2)
     )
     "Float (Alpha)"=1
    )
    Particles.GoldLayerColor:LinearColor=&MakeLinearColorFromVectorAndFloat(
     "Vector (RGB)"=&MergeHelper_VectorSwitch(
      0=&RandomRangeVector(Minimum=<0.75 1.25 1.1>, Maximum=<0.5 2.5 1.25>)
      1=&RandomRangeVector(Minimum=<1.25 1.2 1>, Maximum=<1.1 1.25 1>)
      2=&RandomRangeVector(Minimum=<1.2 0.75 1.5>, Maximum=<1.25 0.5 2>)
      3=?
      Index=$Particles.PartitionParticles.Partition
      Num=::NewEnumerator1
     )
     "Float (Alpha)"=1
    )
   ) {isSetParameters=true}
   /SetVariables(Particles.GlowDynamic:Vector4f=&MakeVector4(X=0.5, Y=&RandomRangeFloat(Minimum=0.15, Maximum=0.2), Z=1, W=1)) {isSetParameters=true}
   &SpriteFacingAndAlignment(
    Sprite Alignment=&ReturnMeshOrientationAxis(Mesh Orientation Axis=::XAxis, Mesh Orientation Quat=$Particles.MeshOrientation)
    WriteAlignment=true
    WriteFacing=false
   )
   /SetVariables(Particles.MeshOrientationGeoLine:Quat4f=&MultiplyQuaternion(
     Quaternion A=$Particles.MeshOrientation
     Quaternion B=&MakeQuaternion(
      Quaternion From=0
      Axis=<1 0 0>
      Angle=45
      From=?
      To=?
      Stabilization Mode:ENiagara_OrientVectorCount=::NewEnumerator0
      Side Direction:Vector3f=?
      Up Direction:Vector3f=?
      Angle Type=::NewEnumerator0
      Normalize:NiagaraBool=true
     )
    )) {isSetParameters=true}
   &SolveForcesAndVelocity(Clamp Velocity=false, Limit Acceleration=false)
  ]
  rendererClasses [&NiagaraMeshRendererProperties, &NiagaraSpriteRendererProperties]
  renderers [
   {rendererClass=&NiagaraMeshRendererProperties, rendererIndex=0}
   {rendererClass=&NiagaraMeshRendererProperties, rendererIndex=1}
   {rendererClass=&NiagaraSpriteRendererProperties, rendererIndex=2}
  ]
 }}
summary {
 systemName=NS_Character_Body_DyeBlast_Idle
 emitters [{emitterName=GeoShapes, rendererClasses [&NiagaraMeshRendererProperties, &NiagaraSpriteRendererProperties]}]
 userVariables []
}
